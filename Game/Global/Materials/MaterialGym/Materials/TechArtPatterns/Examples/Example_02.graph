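# Example Material Gym subsurface material that draws an animated tiled hexagon pattern with colored outer and inner shapes, flattened normals and emissive color.
# MF_TextureUVs uses a hexagon UV texture to tile distance-field samples, adding parallax and a rotator to the inner layer.
# Time-scrolled blurred noise, folded through Frac and Abs, drives hexagon thresholds and parallax height.
# SmoothStep cuts the masks, which are combined and tinted by color parameters.
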Material /Game/Global/Materials/MaterialGym/Materials/TechArtPatterns/Examples/Example_02
graph {
 properties {
  bUsedWithMeshDeformer=true
  bUsedWithSkeletalMesh=true
  pixelDepthOffsetMode=PDOM_Legacy
  shadingModel=MSM_Subsurface
  usedShadingModels=Subsurface
 }
 nodes [
  Constant3Vector()
  n1 = NamedRerouteUsage()
  n2 = NamedRerouteUsage()
  NamedRerouteDeclaration(None=SmoothStep(Min=@n1, Max=Add(A=@n1, B=@n2) {constB=0.05}, Value=NamedRerouteUsage())) {name=SolidHexagon, nodeColor=<rgba 1 0.0469 0 1>}
  n3 = MaterialFunctionCall!(
   Texture=NamedRerouteUsage()
   TilingXY=NamedRerouteUsage()
   OffsetXY
   AnimateOffset
   UseParallax=StaticBool() {value=true}
   Height=NamedRerouteUsage()
  ) {materialFunction=/Game/Global/Materials/MaterialGym/Materials/TechArtPatterns/MaterialFunctions/MF_TextureUVs}
  n4 = TextureSample(UVs=@n3.BaseUVs) {samplerType=SAMPLERTYPE_LinearGrayscale, texture=/Game/Global/Materials/MaterialGym/Materials/TechArtPatterns/Textures/T_hexagons_DF}
  NamedRerouteDeclaration(None=SmoothStep(Value=@n4.R) {constMax=0.21, constMin=0.2}) {name=SolidHexagonMask, nodeColor=<rgba 0.9453 0 1 1>}
  NamedRerouteDeclaration(None=Multiply(A=Time(), B=ScalarParameter() {defaultValue=0.15, parameterName=Speed}) {constB=0.25}) {name=NoiseSpeed, nodeColor=<rgba 0.6406 1 0 1>}
  n5 = MaterialFunctionCall(Texture=NamedRerouteUsage(), TilingXY=NamedRerouteUsage()) {materialFunction=&MF_TextureUVs}
  n6 = TextureSample(UVs=@n5.BaseUVs) {samplerType=SAMPLERTYPE_LinearGrayscale, texture=&T_hexagons_DF}
  NamedRerouteDeclaration(None=@n6.R) {name=OuterHexagonDFSample, nodeColor=<rgba 1 0 0.4687 1>}
  NamedRerouteDeclaration(None=MaterialFunctionCall(
    Normal=TextureSample(UVs=@n5.BaseUVs) {samplerType=SAMPLERTYPE_Normal, texture=/Game/Global/Materials/MaterialGym/Materials/TechArtPatterns/Textures/T_hexagons_N}.RGB
    Flatness=ScalarParameter() {defaultValue=0.8, parameterName=NormalFlatness}
   ) {materialFunction=&FlattenNormal}.Result) {name=OUTNormal, nodeColor=<rgba 0 0.2266 1 1>}
  n7 = NamedRerouteUsage()
  n8 = Constant()
  n9 = TextureSample(UVs=MaterialFunctionCall(
    UVs=@n3.BaseUVs
    Rotation Center=Multiply(A=@n7, B=@n3.TextureUVs)
    "Rotation Angle (0-1)"=MaterialFunctionCall(Speed=@n8) {materialFunction=&TimeWithSpeedVariable}.Time
   ) {materialFunction=&CustomRotator}.Rotated Values) {samplerType=SAMPLERTYPE_LinearGrayscale, texture=&T_hexagons_DF}
  NamedRerouteDeclaration(None=@n9.R) {name=InnerHexagonDFSample, nodeColor=<rgba 1 0 0.6562 1>}
  n10 = NamedRerouteUsage()
  n11 = ScalarParameter() {defaultValue=-3, parameterName=InnerHexagonParallaxHeight}
  NamedRerouteDeclaration(None=Add(A=Multiply(A=@n10, B=@n11) {constB=-2}, B=ScalarParameter() {defaultValue=1.5, parameterName=InnerHexagonParallaxHeightBias}) {constB=0.5}) {name=ParallaxHeight, nodeColor=<rgba 0.9688 1 0 1>}
  NamedRerouteDeclaration(None=ComponentMask(None=VectorParameter() {parameterName=Tiling, defaultValue=<rgba 6 10 0 1>}.RGB) {g=true, r=true}) {name=Tiling, nodeColor=<rgba 1 0.0234 0 1>}
  NamedRerouteDeclaration(None=TextureObject() {samplerType=SAMPLERTYPE_LinearColor, texture=/Game/Global/Materials/MaterialGym/Materials/TechArtPatterns/Textures/T_hexagons_UV}) {name=UVTexture, nodeColor=<rgba 1 0.9141 0 1>}
  NamedRerouteDeclaration(None=Add(A=NamedRerouteUsage(), B=NamedRerouteUsage())) {name=OUTEmissive, nodeColor=<rgba 0.2891 1 0 1>}
  NamedRerouteDeclaration(None=@n3.TextureUVs) {name=InnerHexagonTextureUV, nodeColor=<rgba 0.7578 0 1 1>}
  NamedRerouteDeclaration(None=@n5.TextureUVs) {name=OuterHexagonTextureUVs, nodeColor=<rgba 0.4297 0 1 1>}
  NamedRerouteDeclaration(None=ScalarParameter() {defaultValue=1, parameterName=NoiseContrast}) {name=NoiseContrast, nodeColor=<rgba 0 1 0.1328 1>}
  n12 = Multiply(A=TextureSample(UVs=NamedRerouteUsage()) {samplerType=SAMPLERTYPE_Grayscale, texture=&LowResBlurredNoise}.R) {constB=2}
  n13 = NamedRerouteUsage()
  n14 = ConstantBiasScale(None=Frac(None=Add(A=@n12, B=@n13) {constB=0.5})) {bias=-0.5, scale=2}
  n15 = MaterialFunctionCall(In=Saturate(None=Abs(None=@n14)), Contrast=NamedRerouteUsage()) {materialFunction=&CheapContrast}
  n16 = ScalarParameter() {defaultValue=0.5, parameterName=InnerHexagonInverseScale}
  NamedRerouteDeclaration(None=Saturate(None=Add(A=@n15.Result, B=@n16) {constB=0.5})) {name=InnerHexagonNoise, nodeColor=<rgba 0.6875 1 0 1>}
  n17 = Multiply(A=TextureSample(UVs=NamedRerouteUsage()) {samplerType=SAMPLERTYPE_Grayscale, texture=&LowResBlurredNoise}.R) {constB=2}
  n18 = NamedRerouteUsage()
  n19 = ConstantBiasScale(None=Frac(None=Add(A=@n17, B=@n18) {constB=0.5})) {bias=-0.5, scale=2}
  n20 = MaterialFunctionCall(In=Saturate(None=Abs(None=@n19)), Contrast=NamedRerouteUsage()) {materialFunction=&CheapContrast}
  NamedRerouteDeclaration(None=Min(
    A=NamedRerouteDeclaration(None=@n20.Result) {name=InnerHexagonHeightNoise, nodeColor=<rgba 0.8281 0 1 1>}
    B=ScalarParameter() {defaultValue=1, parameterName=OuterHexagonInverseMinSize}
   ) {constB=0.5}) {name=OuterHexagonNoise, nodeColor=<rgba 0 1 0.1094 1>}
  NamedRerouteDeclaration(None=ScalarParameter() {defaultValue=0.01, parameterName=HexagonEdgeSoftness}) {name=HexagonEdgeSoftness, nodeColor=<rgba 1 0 0.2344 1>}
  n21 = Multiply(A=NamedRerouteUsage(), B=NamedRerouteUsage())
  NamedRerouteDeclaration(None=Multiply(A=Saturate(None=@n21), B=VectorParameter() {parameterName=InnerHexagonColor, defaultValue=<rgba 0.0516 0.2258 0.6562 1>}.RGB)) {name=SolidHexagonColor, nodeColor=<rgba 0 0.2266 1 1>}
  n22 = Subtract(A=NamedRerouteUsage(), B=NamedRerouteUsage())
  NamedRerouteDeclaration(None=Multiply(A=Saturate(None=@n22), B=VectorParameter() {parameterName=OuterHexagonColor, defaultValue=<rgba 0.0158 0.5052 0.2762 1>}.RGB)) {name=OutlineHexagon, nodeColor=<rgba 0 1 0.5547 1>}
  n23 = ComponentMask(None=Add(A=NamedRerouteUsage()) {constB=0.08}) {r=true}
  n24 = NamedRerouteUsage()
  NamedRerouteDeclaration(None=SmoothStep(Min=@n23, Max=Add(A=@n23, B=@n24) {constB=0.01}, Value=NamedRerouteUsage())) {name=InnerHexagon, nodeColor=<rgba 0 1 0.0156 1>}
  n25 = NamedRerouteUsage()
  n26 = NamedRerouteUsage()
  NamedRerouteDeclaration(None=SmoothStep(Min=ComponentMask(None=@n25) {r=true}, Max=Add(A=@n25, B=@n26) {constB=0.01}, Value=NamedRerouteUsage())) {name=OuterHexagon, nodeColor=<rgba 0.8984 0 1 1>}
  NamedRerouteUsage()
  NamedRerouteUsage()
  VectorParameter() {parameterName=Param_1, defaultValue=<rgba 0.0093 0.0238 0.03 1>}
 ]
}
